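# Physics asset for the male medium Underground Rebel Fashion cosmetic body, providing bodies for its dynamic scarf and cloth-volume bones.
# Capsules cover pelvis, spine, thighs and calves, while boxes form most scarf segments and all cloth volumes.
# Asymmetric scarf chains (three segments left, two right) hang from spine_04 through free master joints, stiffer const links and looser segment joints.
# Angularly free constraints also tie scarf links to neighboring cloth volumes.
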PhysicsAsset /BRCosmetics/Characters/Player/Male/Medium/Bodies/M_MED_UndergroundRebel_Fashion/Meshes/M_MED_UndergroundRebel_Fashion_Physics01
bodies {
 calf_l [{shapeType=Capsule, length=35, radius=8, shapeName=calf_l_capsule, center=<-20 -2 0>, rotation=<rot 90 0 0>}]
 calf_r [{shapeType=Capsule, length=35, radius=8, shapeName=calf_r_capsule, center=<20 2 0>, rotation=<rot 90 0 0>}]
 dyn_Cloth_volume_l_A [{
   shapeType=Box
   extentX=18.5
   extentY=5.5
   extentZ=16.5
   shapeName=dyn_Cloth_volume_l_A_box
   center=<-2.9386 -2.6108 5.1121>
   rotation=<rot 0 0 18.9261>
  }]
 dyn_Cloth_volume_l_A01 [{
   shapeType=Box
   extentX=23.5
   extentY=8.5
   extentZ=14.5
   shapeName=dyn_Cloth_volume_l_A01_box
   center=<-6.3353 -2.6108 5.1121>
   rotation=<rot 0 0 18.9261>
  }]
 dyn_Cloth_volume_l_Up02 [{
   shapeType=Box
   extentX=13.5
   extentY=4.5
   extentZ=23.5
   shapeName=dyn_Cloth_volume_l_Up02_box
   center=<0.2205 -2.6108 5.1121>
   rotation=<rot 0 0 18.9261>
  }]
 dyn_Cloth_volume_r_A [{
   shapeType=Box
   extentX=18.5
   extentY=5.5
   extentZ=16.5
   shapeName=dyn_Cloth_volume_r_A_box
   center=<-2.9386 -2.8761 -3.4767>
   rotation=<rot 0 0 -23.1649>
  }]
 dyn_Cloth_volume_r_A01 [{
   shapeType=Box
   extentX=23.5
   extentY=8.5
   extentZ=14.5
   shapeName=dyn_Cloth_volume_r_A01_box
   center=<-6.3353 -2.8761 -3.4767>
   rotation=<rot 0 0 -23.1649>
  }]
 dyn_Cloth_volume_r_Up02 [{
   shapeType=Box
   extentX=13.5
   extentY=4.5
   extentZ=23.5
   shapeName=dyn_Cloth_volume_r_Up02_box
   center=<0.2205 -2.8761 -3.4767>
   rotation=<rot 0 0 -23.1649>
  }]
 dyn_scarf_01_l [{shapeType=Box, extentX=16.5, extentY=3.5, extentZ=5.5, shapeName=dyn_scarf_01_l_box, center=<7.1168 0 0>}]
 dyn_scarf_01_r [{shapeType=Box, extentX=16.5, extentY=3.5, extentZ=5.5, shapeName=dyn_scarf_01_r_box, center=<7.1168 0 0>}]
 dyn_scarf_02_l [{shapeType=Box, extentX=16.5, extentY=3.5, extentZ=7.5, shapeName=dyn_scarf_02_l_box, center=<7.1168 0 0>}]
 dyn_scarf_02_r [{shapeType=Box, extentX=16.5, extentY=3.5, extentZ=7.5, shapeName=dyn_scarf_02_r_box, center=<7.1168 0 0>}]
 dyn_scarf_03_l [{shapeType=Box, extentX=16.5, extentY=3.5, extentZ=7.5, shapeName=dyn_scarf_03_l_box, center=<7.1168 0 0>}]
 dyn_scarf_const_l [{shapeType=Capsule, length=13, radius=2, shapeName=dyn_scarf_const_l_capsule, center=<5.966 0 0.0611>, rotation=<rot 89.4128 -180 -179.9997>}]
 dyn_scarf_const_r [{
   shapeType=Capsule
   length=13
   radius=2
   shapeName=dyn_scarf_const_r_capsule
   center=<5.966 0 0.0611>
   rotation=<rot 89.413 179.9998 -179.9999>
  }]
 dyn_scarf_mstr_l [{shapeType=Capsule, length=0.505, radius=0.505, shapeName=dyn_scarf_mstr_l_capsule, rotation=<rot 0 -90 0>}]
 dyn_scarf_mstr_r [{shapeType=Capsule, length=0.505, radius=0.505, shapeName=dyn_scarf_mstr_r_capsule, rotation=<rot 0 -90 0>}]
 pelvis [{shapeType=Capsule, length=12.4528, radius=10.7478, shapeName=pelvis_capsule, center=<-3.3036 -0.1938 0>}]
 spine_01 [{shapeType=Capsule, length=10.1895, radius=10.459, shapeName=spine_01_capsule, center=<0 0.8646 0>}]
 spine_02 [{shapeType=Capsule, length=9.6533, radius=10.763, shapeName=spine_02_capsule}]
 spine_03 [{shapeType=Capsule, length=9.562, radius=10.7473, shapeName=spine_03_capsule}]
 spine_04 [{shapeType=Capsule, length=1, radius=1, shapeName=spine_04_capsule, rotation=<rot 90 0 0>}]
 thigh_l [{
   shapeType=Capsule
   length=30.6239
   radius=8.9075
   shapeName=thigh_l_capsule
   center=<-17.3517 0.7429 -0.1687>
   rotation=<rot 85.1676 0.2134 -3.7602>
  }]
 thigh_r [{
   shapeType=Capsule
   length=32.201
   radius=8.9571
   shapeName=thigh_r_capsule
   center=<17.1481 -1.1658 0.6344>
   rotation=<rot 88.6845 179.998 179.9988>
  }]
}
constraints [
 {bone1Name=dyn_scarf_01_l, bone2Name=dyn_scarf_const_l, swing1LimitDegrees=45, swing2LimitDegrees=45, twistLimitDegrees=45}
 {bone1Name=dyn_scarf_02_l, bone2Name=dyn_scarf_01_l, swing1LimitDegrees=45, swing2LimitDegrees=45, twistLimitDegrees=45}
 {bone1Name=dyn_scarf_03_l, bone2Name=dyn_scarf_02_l, swing1LimitDegrees=45, swing2LimitDegrees=45, twistLimitDegrees=45}
 {bone1Name=dyn_scarf_01_r, bone2Name=dyn_scarf_const_r, swing1LimitDegrees=45, swing2LimitDegrees=45, twistLimitDegrees=45}
 {bone1Name=dyn_scarf_02_r, bone2Name=dyn_scarf_01_r, swing1LimitDegrees=45, swing2LimitDegrees=45, twistLimitDegrees=45}
 {bone1Name=dyn_scarf_const_l, bone2Name=dyn_scarf_mstr_l, swing1LimitDegrees=15, swing2LimitDegrees=15, twistLimitDegrees=15}
 {
  bone1Name=dyn_scarf_mstr_l
  bone2Name=spine_04
  swing1LimitDegrees=45
  swing1Motion=Free
  swing2LimitDegrees=45
  swing2Motion=Free
  twistLimitDegrees=45
  twistMotion=Free
 }
 {bone1Name=dyn_scarf_const_r, bone2Name=dyn_scarf_mstr_r, swing1LimitDegrees=15, swing2LimitDegrees=15, twistLimitDegrees=15}
 {
  bone1Name=dyn_scarf_mstr_r
  bone2Name=spine_04
  swing1LimitDegrees=45
  swing1Motion=Free
  swing2LimitDegrees=45
  swing2Motion=Free
  twistLimitDegrees=45
  twistMotion=Free
 }
 {bone1Name=dyn_Cloth_volume_r_A01, bone2Name=dyn_Cloth_volume_r_A, swing1LimitDegrees=45, swing2LimitDegrees=45, twistLimitDegrees=45}
 {bone1Name=dyn_Cloth_volume_l_A01, bone2Name=dyn_Cloth_volume_l_A, swing1LimitDegrees=45, swing2LimitDegrees=45, twistLimitDegrees=45}
 {
  bone1Name=dyn_scarf_01_l
  bone2Name=dyn_Cloth_volume_l_A
  swing1LimitDegrees=45
  swing1Motion=Free
  swing2LimitDegrees=45
  swing2Motion=Free
  twistLimitDegrees=45
  twistMotion=Free
 }
 {
  bone1Name=dyn_scarf_const_l
  bone2Name=dyn_Cloth_volume_l_Up02
  swing1LimitDegrees=45
  swing1Motion=Free
  swing2LimitDegrees=45
  swing2Motion=Free
  twistLimitDegrees=45
  twistMotion=Free
 }
 {
  bone1Name=dyn_scarf_01_r
  bone2Name=dyn_Cloth_volume_r_A
  swing1LimitDegrees=45
  swing1Motion=Free
  swing2LimitDegrees=45
  swing2Motion=Free
  twistLimitDegrees=45
  twistMotion=Free
 }
 {
  bone1Name=dyn_scarf_const_r
  bone2Name=dyn_Cloth_volume_r_Up02
  swing1LimitDegrees=45
  swing1Motion=Free
  swing2LimitDegrees=45
  swing2Motion=Free
  twistLimitDegrees=45
  twistMotion=Free
 }
]
